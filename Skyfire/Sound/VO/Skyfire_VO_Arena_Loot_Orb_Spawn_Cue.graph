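# Sound cue for the Skyfire arena announcer VO when a loot orb spawns, on the narrator sound class with high priority.
# After a fixed root delay, a mixer layers random PA feedback with a random announcement line that follows a short random delay.
# Quality-level nodes give the first two levels nine lines plus feedback, and the last two only four lines without feedback.
# It uses the Skyfire VO concurrency group.
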SoundCue /Skyfire/Sound/VO/Skyfire_VO_Arena_Loot_Orb_Spawn_Cue
root=Delay_2
nodes [
 {id=Delay_0, children [Random_0], properties {delayMax=2, delayMin=1.5}}
 {id=Delay_1, children [Random_3]}
 {id=Delay_2, children [Mixer_1], properties {delayMax=5.5, delayMin=5.5}}
 {id=Delay_3, children [Random_1], properties {delayMax=2, delayMin=1.5}}
 {id=Mixer_1, children [QualityLevel_0, QualityLevel_1], properties {inputVolume [1.8,1.25]}}
 {id=QualityLevel_0, children [Delay_0,Delay_0,Delay_3,Delay_3]}
 {id=QualityLevel_1, children [Delay_1,Delay_1,null,null]}
 {
  id=Random_0
  children [WavePlayer_1, WavePlayer_2, WavePlayer_3, WavePlayer_4, WavePlayer_46, WavePlayer_47, WavePlayer_48, WavePlayer_49, WavePlayer_50]
  properties {bShouldExcludeFromBranchCulling=true, weights [1,1,1,1,1,1,1,1,1]}
 }
 {
  id=Random_1
  children [WavePlayer_51, WavePlayer_52, WavePlayer_53, WavePlayer_54]
  properties {bShouldExcludeFromBranchCulling=true, weights [1,1,1,1]}
 }
 {
  id=Random_3
  children [
   WavePlayer_27
   WavePlayer_28
   WavePlayer_29
   WavePlayer_30
   WavePlayer_31
   WavePlayer_32
   WavePlayer_33
   WavePlayer_34
   WavePlayer_35
   WavePlayer_36
  ]
  properties {bShouldExcludeFromBranchCulling=true, weights [1,1,1,1,1,1,1,1,1,1]}
 }
 {id=WavePlayer_1, properties {soundWaveAssetPtr=&/Skyfire_Announcement_Arena_Loot_Orb_Spawn_01}}
 {id=WavePlayer_2, properties {soundWaveAssetPtr=&/Skyfire_Announcement_Arena_Loot_Orb_Spawn_02}}
 {id=WavePlayer_27, properties {soundWaveAssetPtr=/Skyfire/Sound/Skyfire_Arena_PA_Feedback_10}}
 {id=WavePlayer_28, properties {soundWaveAssetPtr=/Skyfire/Sound/Skyfire_Arena_PA_Feedback_01}}
 {id=WavePlayer_29, properties {soundWaveAssetPtr=/Skyfire/Sound/Skyfire_Arena_PA_Feedback_02}}
 {id=WavePlayer_3, properties {soundWaveAssetPtr=&/Skyfire_Announcement_Arena_Loot_Orb_Spawn_03}}
 {id=WavePlayer_30, properties {soundWaveAssetPtr=/Skyfire/Sound/Skyfire_Arena_PA_Feedback_03}}
 {id=WavePlayer_31, properties {soundWaveAssetPtr=/Skyfire/Sound/Skyfire_Arena_PA_Feedback_04}}
 {id=WavePlayer_32, properties {soundWaveAssetPtr=/Skyfire/Sound/Skyfire_Arena_PA_Feedback_05}}
 {id=WavePlayer_33, properties {soundWaveAssetPtr=/Skyfire/Sound/Skyfire_Arena_PA_Feedback_06}}
 {id=WavePlayer_34, properties {soundWaveAssetPtr=/Skyfire/Sound/Skyfire_Arena_PA_Feedback_07}}
 {id=WavePlayer_35, properties {soundWaveAssetPtr=/Skyfire/Sound/Skyfire_Arena_PA_Feedback_08}}
 {id=WavePlayer_36, properties {soundWaveAssetPtr=/Skyfire/Sound/Skyfire_Arena_PA_Feedback_09}}
 {id=WavePlayer_4, properties {soundWaveAssetPtr=&/Skyfire_Announcement_Arena_Loot_Orb_Spawn_04}}
 {id=WavePlayer_46, properties {soundWaveAssetPtr=&/Skyfire_Announcement_Arena_Loot_Orb_Spawn_05}}
 {id=WavePlayer_47, properties {soundWaveAssetPtr=&/Skyfire_Announcement_Arena_Loot_Orb_Spawn_06}}
 {id=WavePlayer_48, properties {soundWaveAssetPtr=&/Skyfire_Announcement_Arena_Loot_Orb_Spawn_07}}
 {id=WavePlayer_49, properties {soundWaveAssetPtr=&/Skyfire_Announcement_Arena_Loot_Orb_Spawn_08}}
 {id=WavePlayer_50, properties {soundWaveAssetPtr=&/Skyfire_Announcement_Arena_Loot_Orb_Spawn_09}}
 {id=WavePlayer_51, properties {soundWaveAssetPtr=&Skyfire_Announcement_Arena_Loot_Orb_Spawn_01}}
 {id=WavePlayer_52, properties {soundWaveAssetPtr=&Skyfire_Announcement_Arena_Loot_Orb_Spawn_02}}
 {id=WavePlayer_53, properties {soundWaveAssetPtr=&Skyfire_Announcement_Arena_Loot_Orb_Spawn_03}}
 {id=WavePlayer_54, properties {soundWaveAssetPtr=&Skyfire_Announcement_Arena_Loot_Orb_Spawn_04}}
]
properties {
 bBypassVolumeScaleForPriority=true
 maxDistance=2097152
 priority=90
 soundClassObject=/Game/Sounds/FortSoundClassesAndModes/Fort_VO_Narrator
 volumeMultiplier=0.6667
 concurrencyOverrides {maxCount=1, resolutionRule=PreventNew}
 concurrencySet [&/Skyfire_VO_Concurrency_Group]
}
