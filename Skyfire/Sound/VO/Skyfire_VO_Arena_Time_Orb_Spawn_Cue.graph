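# Sound Cue for the Skyfire arena announcer voice line that plays when a Time Orb spawns, routed to the narrator sound class at raised priority.
# A Mixer layers a random announcement, delayed one second, over a random PA feedback sound.
# Quality Level nodes give the first two quality slots six announcement variants plus feedback, and the last two a separate three-variant set without feedback.
# Concurrency settings block overlapping new instances.
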SoundCue /Skyfire/Sound/VO/Skyfire_VO_Arena_Time_Orb_Spawn_Cue
root=Mixer_1
nodes [
 {id=Delay_2, children [Random_4]}
 {id=Delay_3, children [Random_0], properties {delayMax=1, delayMin=1}}
 {id=Delay_6, children [Random_6], properties {delayMax=1, delayMin=1}}
 {id=Mixer_1, children [QualityLevel_0, QualityLevel_2], properties {inputVolume [1.8,1.25]}}
 {id=QualityLevel_0, children [Delay_3,Delay_3,Delay_6,Delay_6]}
 {id=QualityLevel_2, children [Delay_2,Delay_2,null,null]}
 {
  id=Random_0
  children [WavePlayer_1, WavePlayer_2, WavePlayer_3, WavePlayer_4, WavePlayer_46, WavePlayer_47]
  properties {bShouldExcludeFromBranchCulling=true, weights [1,1,1,1,1,1]}
 }
 {
  id=Random_4
  children [
   WavePlayer_54
   WavePlayer_55
   WavePlayer_56
   WavePlayer_57
   WavePlayer_58
   WavePlayer_59
   WavePlayer_60
   WavePlayer_61
   WavePlayer_62
   WavePlayer_63
  ]
  properties {bShouldExcludeFromBranchCulling=true, weights [1,1,1,1,1,1,1,1,1,1]}
 }
 {id=Random_6, children [WavePlayer_64, WavePlayer_65, WavePlayer_66], properties {bShouldExcludeFromBranchCulling=true, weights [1,1,1]}}
 {id=WavePlayer_1, properties {soundWaveAssetPtr=&/Skyfire_Announcement_Arena_Time_Orb_Spawn_01}}
 {id=WavePlayer_2, properties {soundWaveAssetPtr=&/Skyfire_Announcement_Arena_Time_Orb_Spawn_02}}
 {id=WavePlayer_3, properties {soundWaveAssetPtr=&/Skyfire_Announcement_Arena_Time_Orb_Spawn_03}}
 {id=WavePlayer_4, properties {soundWaveAssetPtr=&/Skyfire_Announcement_Arena_Time_Orb_Spawn_04}}
 {id=WavePlayer_46, properties {soundWaveAssetPtr=&/Skyfire_Announcement_Arena_Time_Orb_Spawn_05}}
 {id=WavePlayer_47, properties {soundWaveAssetPtr=&/Skyfire_Announcement_Arena_Time_Orb_Spawn_06}}
 {id=WavePlayer_54, properties {soundWaveAssetPtr=/Skyfire/Sound/Skyfire_Arena_PA_Feedback_10}}
 {id=WavePlayer_55, properties {soundWaveAssetPtr=/Skyfire/Sound/Skyfire_Arena_PA_Feedback_01}}
 {id=WavePlayer_56, properties {soundWaveAssetPtr=/Skyfire/Sound/Skyfire_Arena_PA_Feedback_02}}
 {id=WavePlayer_57, properties {soundWaveAssetPtr=/Skyfire/Sound/Skyfire_Arena_PA_Feedback_03}}
 {id=WavePlayer_58, properties {soundWaveAssetPtr=/Skyfire/Sound/Skyfire_Arena_PA_Feedback_04}}
 {id=WavePlayer_59, properties {soundWaveAssetPtr=/Skyfire/Sound/Skyfire_Arena_PA_Feedback_05}}
 {id=WavePlayer_60, properties {soundWaveAssetPtr=/Skyfire/Sound/Skyfire_Arena_PA_Feedback_06}}
 {id=WavePlayer_61, properties {soundWaveAssetPtr=/Skyfire/Sound/Skyfire_Arena_PA_Feedback_07}}
 {id=WavePlayer_62, properties {soundWaveAssetPtr=/Skyfire/Sound/Skyfire_Arena_PA_Feedback_08}}
 {id=WavePlayer_63, properties {soundWaveAssetPtr=/Skyfire/Sound/Skyfire_Arena_PA_Feedback_09}}
 {id=WavePlayer_64, properties {soundWaveAssetPtr=&Skyfire_Announcement_Arena_Time_Orb_Spawn_01}}
 {id=WavePlayer_65, properties {soundWaveAssetPtr=&Skyfire_Announcement_Arena_Time_Orb_Spawn_02}}
 {id=WavePlayer_66, properties {soundWaveAssetPtr=&Skyfire_Announcement_Arena_Time_Orb_Spawn_03}}
]
properties {
 bBypassVolumeScaleForPriority=true
 maxDistance=2097152
 priority=90
 soundClassObject=/Game/Sounds/FortSoundClassesAndModes/Fort_VO_Narrator
 volumeMultiplier=0.6667
 concurrencyOverrides {maxCount=1, resolutionRule=PreventNew}
 concurrencySet [&/Skyfire_VO_Concurrency_Group]
}
