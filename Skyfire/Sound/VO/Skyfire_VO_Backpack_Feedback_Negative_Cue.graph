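# Sound cue that plays negative feedback announcer voice lines for Skyfire, spatialized with backpack attenuation as narrator VO.
# A mixer layers a pitch-randomized walkie-talkie-on sound with one of six equally weighted lines, delayed briefly and mixed louder.
# Only the first quality-level slot is wired, and the random branch is excluded from branch culling.
# A Skyfire VO concurrency group governs simultaneous playback, with virtualization disabled.
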SoundCue /Skyfire/Sound/VO/Skyfire_VO_Backpack_Feedback_Negative_Cue
root=QualityLevel_0
nodes [
 {id=Delay_0, children [Random_0], properties {delayMax=0.2, delayMin=0.2}}
 {id=Mixer_0, children [Modulator_0,Delay_0], properties {inputVolume [0.75,1.35]}}
 {id=Modulator_0, children [WavePlayer_12], properties {pitchMax=1.1, pitchMin=0.85, volumeMax=1, volumeMin=1}}
 {id=QualityLevel_0, children [Mixer_0,null,null,null]}
 {
  id=Random_0
  children [WavePlayer_1,WavePlayer_2,WavePlayer_3,WavePlayer_4,WavePlayer_5,WavePlayer_6]
  properties {bShouldExcludeFromBranchCulling=true, weights [1,1,1,1,1,1]}
 }
 {id=WavePlayer_1, properties {soundWaveAssetPtr=&/Skyfire_Announcement_Feedback_Negative_01}}
 {id=WavePlayer_12, properties {soundWaveAssetPtr=&/Skyfire_Announcement_Walkie_Talkie_On_01}}
 {id=WavePlayer_2, properties {soundWaveAssetPtr=&/Skyfire_Announcement_Feedback_Negative_02}}
 {id=WavePlayer_3, properties {soundWaveAssetPtr=&/Skyfire_Announcement_Feedback_Negative_03}}
 {id=WavePlayer_4, properties {soundWaveAssetPtr=&/Skyfire_Announcement_Feedback_Negative_04}}
 {id=WavePlayer_5, properties {soundWaveAssetPtr=&/Skyfire_Announcement_Feedback_Negative_05}}
 {id=WavePlayer_6, properties {soundWaveAssetPtr=&/Skyfire_Announcement_Feedback_Negative_06}}
]
properties {
 attenuationSettings=/Skyfire/Sound/Skyfire_Backpack_Att
 bBypassVolumeScaleForPriority=true
 maxDistance=1350
 priority=0
 soundClassObject=/Game/Sounds/FortSoundClassesAndModes/Fort_VO_Narrator
 virtualizationMode=Disabled
 volumeMultiplier=0.8
 concurrencySet [&/Skyfire_VO_Concurrency_Group]
}
