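# Sound cue for the Skyfire backpack's loot orb collect voice-over: a walkie-talkie click followed by an announcer line.
# A mixer layers the pitch-randomized click with a briefly delayed quality switch, whose first two levels randomize ten lines and last two randomize four.
# The root quality switch connects this mix only to its first slot, leaving the other three empty.
# It plays as narrator VO with backpack attenuation, shared VO concurrency and virtualization disabled.
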SoundCue /Skyfire/Sound/VO/Skyfire_VO_Backpack_Loot_Orb_Collect_Cue
root=QualityLevel_1
nodes [
 {id=Delay_0, children [QualityLevel_0], properties {delayMax=0.2, delayMin=0.2}}
 {id=Mixer_0, children [Modulator_0,Delay_0], properties {inputVolume [0.75,1.35]}}
 {id=Modulator_0, children [WavePlayer_12], properties {pitchMax=1.1, pitchMin=0.85, volumeMax=1, volumeMin=1}}
 {id=QualityLevel_0, children [Random_1,Random_1,Random_2,Random_2]}
 {id=QualityLevel_1, children [Mixer_0,null,null,null]}
 {
  id=Random_1
  children [
   WavePlayer_2
   WavePlayer_13
   WavePlayer_14
   WavePlayer_15
   WavePlayer_16
   WavePlayer_17
   WavePlayer_18
   WavePlayer_19
   WavePlayer_20
   WavePlayer_21
  ]
  properties {bShouldExcludeFromBranchCulling=true, weights [1,1,1,1,1,1,1,1,1,1]}
 }
 {
  id=Random_2
  children [WavePlayer_22, WavePlayer_24, WavePlayer_25, WavePlayer_26]
  properties {bShouldExcludeFromBranchCulling=true, weights [1,1,1,1]}
 }
 {id=WavePlayer_12, properties {soundWaveAssetPtr=&/Skyfire_Announcement_Walkie_Talkie_On_01}}
 {id=WavePlayer_13, properties {soundWaveAssetPtr=&/Skyfire_Announcement_Arena_Loot_Orb_Collect_01}}
 {id=WavePlayer_14, properties {soundWaveAssetPtr=&/Skyfire_Announcement_Arena_Loot_Orb_Collect_02}}
 {id=WavePlayer_15, properties {soundWaveAssetPtr=&/Skyfire_Announcement_Arena_Loot_Orb_Collect_03}}
 {id=WavePlayer_16, properties {soundWaveAssetPtr=&/Skyfire_Announcement_Arena_Loot_Orb_Collect_04}}
 {id=WavePlayer_17, properties {soundWaveAssetPtr=&/Skyfire_Announcement_Arena_Loot_Orb_Collect_05}}
 {id=WavePlayer_18, properties {soundWaveAssetPtr=&/Skyfire_Announcement_Arena_Loot_Orb_Collect_06}}
 {id=WavePlayer_19, properties {soundWaveAssetPtr=&/Skyfire_Announcement_Arena_Loot_Orb_Collect_07}}
 {id=WavePlayer_2, properties {soundWaveAssetPtr=&/Skyfire_Announcement_Arena_Loot_Orb_Collect_10}}
 {id=WavePlayer_20, properties {soundWaveAssetPtr=&/Skyfire_Announcement_Arena_Loot_Orb_Collect_08}}
 {id=WavePlayer_21, properties {soundWaveAssetPtr=&/Skyfire_Announcement_Arena_Loot_Orb_Collect_09}}
 {id=WavePlayer_22, properties {soundWaveAssetPtr=&Skyfire_Announcement_Arena_Loot_Orb_Collect_10}}
 {id=WavePlayer_24, properties {soundWaveAssetPtr=&Skyfire_Announcement_Arena_Loot_Orb_Collect_01}}
 {id=WavePlayer_25, properties {soundWaveAssetPtr=&Skyfire_Announcement_Arena_Loot_Orb_Collect_02}}
 {id=WavePlayer_26, properties {soundWaveAssetPtr=&Skyfire_Announcement_Arena_Loot_Orb_Collect_03}}
]
properties {
 attenuationSettings=/Skyfire/Sound/Skyfire_Backpack_Att
 bBypassVolumeScaleForPriority=true
 maxDistance=1350
 priority=0
 soundClassObject=/Game/Sounds/FortSoundClassesAndModes/Fort_VO_Narrator
 virtualizationMode=Disabled
 volumeMultiplier=0.8
 concurrencySet [&/Skyfire_VO_Concurrency_Group]
}
