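# Sound cue for the Skyfire voice-over tied to backpack time expiring, routed through the Fort narrator sound class.
# The node graph is empty with no root, so the cue references no audio as it stands.
# It carries only playback settings: Skyfire backpack attenuation, a shared VO concurrency group, reduced volume and disabled virtualization.
# Priority is zero and bypasses volume scaling.
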SoundCue /Skyfire/Sound/VO/Skyfire_VO_Backpack_Time_Expires_Cue
root=null
nodes []
properties {
 attenuationSettings=/Skyfire/Sound/Skyfire_Backpack_Att
 bBypassVolumeScaleForPriority=true
 maxDistance=1350
 priority=0
 soundClassObject=/Game/Sounds/FortSoundClassesAndModes/Fort_VO_Narrator
 virtualizationMode=Disabled
 volumeMultiplier=0.8
 concurrencySet [&/Skyfire_VO_Concurrency_Group]
}
