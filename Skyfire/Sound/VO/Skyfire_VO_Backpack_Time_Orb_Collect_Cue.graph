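# Sound cue for Skyfire backpack narrator voice-over announcing an arena Time Orb collection.
# An equal-weight random gate either plays nothing or mixes a pitch-varied walkie-talkie click with one of six random voice lines, slightly delayed and louder, mimicking a radio call.
# Only the first quality-level branch is wired.
# Playback uses backpack attenuation, the Skyfire VO concurrency group and disabled virtualization.
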SoundCue /Skyfire/Sound/VO/Skyfire_VO_Backpack_Time_Orb_Collect_Cue
root=QualityLevel_0
nodes [
 {id=Delay_0, children [Random_1], properties {delayMax=0.2, delayMin=0.2}}
 {id=Mixer_0, children [Modulator_0,Delay_0], properties {inputVolume [0.75,1.35]}}
 {id=Modulator_0, children [WavePlayer_12], properties {pitchMax=1.1, pitchMin=0.85, volumeMax=1, volumeMin=1}}
 {id=QualityLevel_0, children [Random_0,null,null,null]}
 {id=Random_0, children [Mixer_0,null], properties {weights [1,1]}}
 {
  id=Random_1
  children [WavePlayer_0, WavePlayer_7, WavePlayer_8, WavePlayer_9, WavePlayer_10, WavePlayer_11]
  properties {bShouldExcludeFromBranchCulling=true, weights [1,1,1,1,1,1]}
 }
 {id=WavePlayer_0, properties {soundWaveAssetPtr=&/Skyfire_Announcement_Arena_Time_Orb_Collect_01}}
 {id=WavePlayer_10, properties {soundWaveAssetPtr=&/Skyfire_Announcement_Arena_Time_Orb_Collect_05}}
 {id=WavePlayer_11, properties {soundWaveAssetPtr=&/Skyfire_Announcement_Arena_Time_Orb_Collect_06}}
 {id=WavePlayer_12, properties {soundWaveAssetPtr=&/Skyfire_Announcement_Walkie_Talkie_On_01}}
 {id=WavePlayer_7, properties {soundWaveAssetPtr=&/Skyfire_Announcement_Arena_Time_Orb_Collect_02}}
 {id=WavePlayer_8, properties {soundWaveAssetPtr=&/Skyfire_Announcement_Arena_Time_Orb_Collect_03}}
 {id=WavePlayer_9, properties {soundWaveAssetPtr=&/Skyfire_Announcement_Arena_Time_Orb_Collect_04}}
]
properties {
 attenuationSettings=/Skyfire/Sound/Skyfire_Backpack_Att
 maxDistance=1350
 priority=0
 soundClassObject=/Game/Sounds/FortSoundClassesAndModes/Fort_VO_Narrator
 virtualizationMode=Disabled
 volumeMultiplier=0.8
 concurrencySet [&/Skyfire_VO_Concurrency_Group]
}
