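# Sound cue for the Skyfire loot room reveal announcement, played as narrator-class, high-priority VO with concurrency capped at one instance.
# After a fixed delay, it mixes a randomized, slightly delayed voice line with quieter pitched-up random PA feedback and an announcement click.
# Quality-level nodes trim the content: the last two tiers draw from fewer voice variants and drop feedback, while the click plays only in the first tier.
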SoundCue /Skyfire/Sound/VO/Skyfire_VO_Loot_Rm_Reveal_Cue
root=Delay_1
nodes [
 {id=Attenuation_0, children [Modulator_0]}
 {id=Attenuation_2, children [WavePlayer_19]}
 {id=Delay_0, children [QualityLevel_0], properties {delayMax=0.75, delayMin=0.5}}
 {id=Delay_1, children [Mixer_1], properties {delayMax=3, delayMin=3}}
 {id=Mixer_1, children [Delay_0, Attenuation_0, QualityLevel_1], properties {inputVolume [1.2,0.1,0.25]}}
 {id=Modulator_0, children [QualityLevel_3], properties {pitchMax=1.2, pitchMin=1.2, volumeMax=1, volumeMin=1}}
 {id=QualityLevel_0, children [Random_3,Random_3,Random_4,Random_4]}
 {id=QualityLevel_1, children [Attenuation_2, null, null, null]}
 {id=QualityLevel_3, children [Random_9,Random_9,null,null]}
 {
  id=Random_3
  children [WavePlayer_20, WavePlayer_21, WavePlayer_22, WavePlayer_23, WavePlayer_24]
  properties {bShouldExcludeFromBranchCulling=true, weights [1,1,1,1,1]}
 }
 {id=Random_4, children [WavePlayer_25, WavePlayer_26, WavePlayer_27], properties {bShouldExcludeFromBranchCulling=true, weights [1,1,1]}}
 {
  id=Random_9
  children [
   WavePlayer_76
   WavePlayer_77
   WavePlayer_78
   WavePlayer_79
   WavePlayer_80
   WavePlayer_81
   WavePlayer_82
   WavePlayer_83
   WavePlayer_84
   WavePlayer_85
  ]
  properties {bShouldExcludeFromBranchCulling=true, weights [1,1,1,1,1,1,1,1,1,1]}
 }
 {id=WavePlayer_19, properties {soundWaveAssetPtr=&/Skyfire_Prison_Cell_Announcement_Click_01}}
 {id=WavePlayer_20, properties {soundWaveAssetPtr=&/Skyfire_Announcement_Loot_Rm_Reveal_01}}
 {id=WavePlayer_21, properties {soundWaveAssetPtr=&/Skyfire_Announcement_Loot_Rm_Reveal_02}}
 {id=WavePlayer_22, properties {soundWaveAssetPtr=&/Skyfire_Announcement_Loot_Rm_Reveal_03}}
 {id=WavePlayer_23, properties {soundWaveAssetPtr=&/Skyfire_Announcement_Loot_Rm_Reveal_04}}
 {id=WavePlayer_24, properties {soundWaveAssetPtr=&/Skyfire_Announcement_Loot_Rm_Reveal_05}}
 {id=WavePlayer_25, properties {soundWaveAssetPtr=&Skyfire_Announcement_Loot_Rm_Reveal_01}}
 {id=WavePlayer_26, properties {soundWaveAssetPtr=&Skyfire_Announcement_Loot_Rm_Reveal_02}}
 {id=WavePlayer_27, properties {soundWaveAssetPtr=&Skyfire_Announcement_Loot_Rm_Reveal_03}}
 {id=WavePlayer_76, properties {soundWaveAssetPtr=/Skyfire/Sound/Skyfire_Arena_PA_Feedback_10}}
 {id=WavePlayer_77, properties {soundWaveAssetPtr=/Skyfire/Sound/Skyfire_Arena_PA_Feedback_01}}
 {id=WavePlayer_78, properties {soundWaveAssetPtr=/Skyfire/Sound/Skyfire_Arena_PA_Feedback_02}}
 {id=WavePlayer_79, properties {soundWaveAssetPtr=/Skyfire/Sound/Skyfire_Arena_PA_Feedback_03}}
 {id=WavePlayer_80, properties {soundWaveAssetPtr=/Skyfire/Sound/Skyfire_Arena_PA_Feedback_04}}
 {id=WavePlayer_81, properties {soundWaveAssetPtr=/Skyfire/Sound/Skyfire_Arena_PA_Feedback_05}}
 {id=WavePlayer_82, properties {soundWaveAssetPtr=/Skyfire/Sound/Skyfire_Arena_PA_Feedback_06}}
 {id=WavePlayer_83, properties {soundWaveAssetPtr=/Skyfire/Sound/Skyfire_Arena_PA_Feedback_07}}
 {id=WavePlayer_84, properties {soundWaveAssetPtr=/Skyfire/Sound/Skyfire_Arena_PA_Feedback_08}}
 {id=WavePlayer_85, properties {soundWaveAssetPtr=/Skyfire/Sound/Skyfire_Arena_PA_Feedback_09}}
]
properties {
 bBypassVolumeScaleForPriority=true
 bOverrideConcurrency=true
 maxDistance=2097152
 priority=90
 soundClassObject=/Game/Sounds/FortSoundClassesAndModes/Fort_VO_Narrator
 volumeMultiplier=0.6667
 attenuationOverrides {
  bEnableListenerFocus=true
  falloffDistance=5000
  focusAzimuth=20
  nonFocusAzimuth=40
  nonFocusVolumeAttenuation=0.3
  reverbDistanceMax=150
  reverbDistanceMin=100
 }
 concurrencyOverrides {maxCount=1, resolutionRule=PreventNew}
 concurrencySet [null]
}
