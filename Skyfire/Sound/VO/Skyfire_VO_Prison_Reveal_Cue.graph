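# Sound cue for the Skyfire prison cell reveal announcement voice-over, played as a high-priority single instance that blocks new copies.
# After a three-second delay, a mixer layers a quieter announcement click with a slightly delayed, louder random pick from five voice lines, each branch through an attenuation node.
# Quality-level nodes choose between two five-line sets for the voice and keep the click only at the first level.
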SoundCue /Skyfire/Sound/VO/Skyfire_VO_Prison_Reveal_Cue
root=Delay_2
nodes [
 {id=Attenuation_0, children [Modulator_1]}
 {id=Attenuation_1, children [Delay_1]}
 {id=Delay_1, children [QualityLevel_0], properties {delayMax=0.35, delayMin=0.35}}
 {id=Delay_2, children [Mixer_1], properties {delayMax=3, delayMin=3}}
 {id=Mixer_1, children [Attenuation_1, QualityLevel_1], properties {inputVolume [1.2,0.65]}}
 {id=Modulator_1, children [WavePlayer_1], properties {volumeMax=1, volumeMin=1}}
 {id=QualityLevel_0, children [Random_1,Random_1,Random_6,Random_6]}
 {id=QualityLevel_1, children [Attenuation_0, null, null, null]}
 {
  id=Random_1
  children [WavePlayer_49, WavePlayer_50, WavePlayer_51, WavePlayer_52, WavePlayer_53]
  properties {bShouldExcludeFromBranchCulling=true, weights [1,1,1,1,1]}
 }
 {
  id=Random_6
  children [WavePlayer_56, WavePlayer_57, WavePlayer_58, WavePlayer_59, WavePlayer_60]
  properties {bShouldExcludeFromBranchCulling=true, weights [1,1,1,1,1]}
 }
 {id=WavePlayer_1, properties {soundWaveAssetPtr=&/Skyfire_Prison_Cell_Announcement_Click_01}}
 {id=WavePlayer_49, properties {soundWaveAssetPtr=&/Skyfire_Announcement_Cell_Reveal_01}}
 {id=WavePlayer_50, properties {soundWaveAssetPtr=&/Skyfire_Announcement_Cell_Reveal_02}}
 {id=WavePlayer_51, properties {soundWaveAssetPtr=&/Skyfire_Announcement_Cell_Reveal_03}}
 {id=WavePlayer_52, properties {soundWaveAssetPtr=&/Skyfire_Announcement_Cell_Reveal_04}}
 {id=WavePlayer_53, properties {soundWaveAssetPtr=&/Skyfire_Announcement_Cell_Reveal_05}}
 {id=WavePlayer_56, properties {soundWaveAssetPtr=&Skyfire_Announcement_Cell_Reveal_01}}
 {id=WavePlayer_57, properties {soundWaveAssetPtr=&Skyfire_Announcement_Cell_Reveal_02}}
 {id=WavePlayer_58, properties {soundWaveAssetPtr=&Skyfire_Announcement_Cell_Reveal_03}}
 {id=WavePlayer_59, properties {soundWaveAssetPtr=&Skyfire_Announcement_Cell_Reveal_04}}
 {id=WavePlayer_60, properties {soundWaveAssetPtr=&Skyfire_Announcement_Cell_Reveal_05}}
]
properties {
 bBypassVolumeScaleForPriority=true
 bOverrideConcurrency=true
 maxDistance=2097152
 priority=90
 soundClassObject=/Game/Sounds/FortSoundClassesAndModes/Fort_LTM_BR_InGame_Master
 volumeMultiplier=1.25
 attenuationOverrides {
  bEnableFocusInterpolation=true
  bEnableListenerFocus=true
  bEnableOcclusion=true
  binauralRadius=200
  falloffDistance=5000
  focusAttackInterpSpeed=10
  focusAzimuth=20
  focusReleaseInterpSpeed=20
  focusVolumeAttenuation=1.75
  nonFocusAzimuth=40
  nonFocusVolumeAttenuation=0.5
  occlusionLowPassFilterFrequency=4000
  occlusionTraceChannel=ECC_GameTraceChannel11
  reverbDistanceMax=100
  reverbDistanceMin=50
  attenuationShapeExtents {x=0}
 }
 concurrencyOverrides {maxCount=1, resolutionRule=PreventNew}
 concurrencySet [null]
}
